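# Physics asset for the CosmicTrellis medium male player body cosmetic, setting up dynamic bones of its jacket.
# Five capsule chains (center, two per side) hang from the pelvis through small offset bodies, with limited swing and very narrow twist.
# Free-angular constraints cross-link each side's two chains at their last two segments, tying neighboring panels together.
# Pelvis, thigh and calf capsules, plus a very large root capsule, provide body shapes.
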PhysicsAsset /BRCosmetics/Characters/Player/Male/Medium/Bodies/M_MED_CosmicTrellis/Meshes/M_MED_CosmicTrellis_Physics
bodies {
 calf_l [{shapeType=Capsule, length=32.2793, radius=9.096, center=<-26.2814 0.0004 0>, rotation=<rot -90 -0.001 0>}]
 calf_r [{shapeType=Capsule, length=32.2793, radius=9.096, center=<26.2814 -0.0004 0>, rotation=<rot -90 179.999 0>}]
 dyn_jacket_c_02 [{shapeType=Capsule, length=10, radius=5, shapeName=dyn_jacket_c_02_capsule, center=<5 0 4>, rotation=<rot 0 0 90>}]
 dyn_jacket_c_02_offset [{shapeType=Capsule, length=0.505, radius=0.505, shapeName=dyn_jacket_c_02_offset_capsule, rotation=<rot -0.0004 0 90>}]
 dyn_jacket_c_03 [{shapeType=Capsule, length=17, radius=10, shapeName=dyn_jacket_c_03_capsule, center=<10 0 8.5>, rotation=<rot 0 0 90>}]
 dyn_jacket_c_04 [{shapeType=Capsule, length=20, radius=10, shapeName=dyn_jacket_c_04_capsule, center=<10 0 8.5>, rotation=<rot 0 0 90>}]
 dyn_jacket_c_05 [{shapeType=Capsule, length=23, radius=10, shapeName=dyn_jacket_c_05_capsule, center=<10 0 8.5>, rotation=<rot 0 0 90>}]
 dyn_jacket_c_06 [{shapeType=Capsule, length=26, radius=10, shapeName=dyn_jacket_c_06_capsule, center=<10 0 8.5>, rotation=<rot 0 0 90>}]
 dyn_jacket_l_a_02 [{shapeType=Capsule, length=5, radius=5, shapeName=dyn_jacket_l_a_02_capsule, center=<5 -2 3.5>, rotation=<rot 0 0 90>}]
 dyn_jacket_l_a_02_offset [{shapeType=Capsule, length=0.505, radius=0.505, shapeName=dyn_jacket_l_a_02_offset_capsule, rotation=<rot 0.0004 0.0002 89.9999>}]
 dyn_jacket_l_a_03 [{shapeType=Capsule, length=10, radius=10, shapeName=dyn_jacket_l_a_03_capsule, center=<10 -7 9.75>, rotation=<rot 0 0 90>}]
 dyn_jacket_l_a_04 [{shapeType=Capsule, length=10, radius=10, shapeName=dyn_jacket_l_a_04_capsule, center=<10 -7 9.75>, rotation=<rot 0 0 90>}]
 dyn_jacket_l_a_05 [{shapeType=Capsule, length=10, radius=10, shapeName=dyn_jacket_l_a_05_capsule, center=<10 -7 9.75>, rotation=<rot 0 0 90>}]
 dyn_jacket_l_a_06 [{shapeType=Capsule, length=10, radius=10, shapeName=dyn_jacket_l_a_06_capsule, center=<10 -7 6>, rotation=<rot 0 0 90>}]
 dyn_jacket_l_b_02 [{shapeType=Capsule, length=5, radius=5, shapeName=dyn_jacket_l_b_02_capsule, center=<5 0 4.75>, rotation=<rot 0 0 90>}]
 dyn_jacket_l_b_02_offset [{shapeType=Capsule, length=0.505, radius=0.505, shapeName=dyn_jacket_l_b_02_offset_capsule, rotation=<rot 0 0.0004 89.9999>}]
 dyn_jacket_l_b_03 [{shapeType=Capsule, length=10, radius=10, shapeName=dyn_jacket_l_b_03_capsule, center=<10 0 9.75>, rotation=<rot 0 0 90>}]
 dyn_jacket_l_b_04 [
  {shapeType=Capsule, length=10, radius=10, shapeName=dyn_jacket_l_b_06_capsule, center=<10 0 6.25>, rotation=<rot 0 0 90>}
  {
   shapeType=Capsule
   length=7.3506
   radius=8.6279
   shapeName=dyn_jacket_l_b_06_capsule
   center=<10.3689 -15.6759 2.2889>
   rotation=<rot 0 0.0008 -130.0002>
  }
 ]
 dyn_jacket_l_b_05 [
  {shapeType=Capsule, length=10, radius=10, shapeName=dyn_jacket_l_b_06_capsule, center=<10 0 6.25>, rotation=<rot 0 0 90>}
  {
   shapeType=Capsule
   length=7.3506
   radius=8.6279
   shapeName=dyn_jacket_l_b_06_capsule
   center=<10.3689 -15.6759 2.2889>
   rotation=<rot 0 0.0008 -130.0002>
  }
 ]
 dyn_jacket_l_b_06 [
  {shapeType=Capsule, length=10, radius=10, shapeName=dyn_jacket_l_b_06_capsule, center=<10 0 6.25>, rotation=<rot 0 0 90>}
  {
   shapeType=Capsule
   length=7.3506
   radius=8.6279
   shapeName=dyn_jacket_l_b_06_capsule
   center=<10.3689 -15.6759 2.2889>
   rotation=<rot 0 0.0008 -130.0002>
  }
 ]
 dyn_jacket_r_a_02 [{shapeType=Capsule, length=5, radius=5, shapeName=dyn_jacket_r_a_02_capsule, center=<5 2 3.5>, rotation=<rot 0 0 90>}]
 dyn_jacket_r_a_02_offset [{shapeType=Capsule, length=0.505, radius=0.505, shapeName=dyn_jacket_r_a_02_offset_capsule, rotation=<rot 0.0003 -0.0002 90>}]
 dyn_jacket_r_a_03 [{shapeType=Capsule, length=10, radius=10, shapeName=dyn_jacket_r_a_03_capsule, center=<10 7 9.75>, rotation=<rot 0 0 90>}]
 dyn_jacket_r_a_04 [{shapeType=Capsule, length=10, radius=10, shapeName=dyn_jacket_r_a_04_capsule, center=<10 7 9.75>, rotation=<rot 0 0 90>}]
 dyn_jacket_r_a_05 [{shapeType=Capsule, length=10, radius=10, shapeName=dyn_jacket_r_a_05_capsule, center=<10 7 9.75>, rotation=<rot 0 0 90>}]
 dyn_jacket_r_a_06 [{shapeType=Capsule, length=10, radius=10, shapeName=dyn_jacket_r_a_06_capsule, center=<10 7 6>, rotation=<rot 0 0 90>}]
 dyn_jacket_r_b_02 [{shapeType=Capsule, length=5, radius=5, shapeName=dyn_jacket_r_b_02_capsule, center=<5 0 4.75>, rotation=<rot 0 0 90>}]
 dyn_jacket_r_b_02_offset [{shapeType=Capsule, length=0.505, radius=0.505, shapeName=dyn_jacket_r_b_02_offset_capsule, rotation=<rot 0 -0.0004 90.0001>}]
 dyn_jacket_r_b_03 [{shapeType=Capsule, length=10, radius=10, shapeName=dyn_jacket_r_b_03_capsule, center=<10 0 9.75>, rotation=<rot 0 0 90>}]
 dyn_jacket_r_b_04 [
  {shapeType=Capsule, length=10, radius=10, shapeName=dyn_jacket_l_b_06_capsule, center=<10 0 6.25>, rotation=<rot 0 0 90>}
  {
   shapeType=Capsule
   length=7.3506
   radius=8.6279
   shapeName=dyn_jacket_l_b_06_capsule
   center=<10.3689 15.676 2.2889>
   rotation=<rot 0 0.0008 130.0002>
  }
 ]
 dyn_jacket_r_b_05 [
  {shapeType=Capsule, length=10, radius=10, shapeName=dyn_jacket_l_b_06_capsule, center=<10 0 6.25>, rotation=<rot 0 0 90>}
  {
   shapeType=Capsule
   length=7.3506
   radius=8.6279
   shapeName=dyn_jacket_l_b_06_capsule
   center=<10.3689 15.676 2.2889>
   rotation=<rot 0 0.0008 130.0002>
  }
 ]
 dyn_jacket_r_b_06 [
  {shapeType=Capsule, length=10, radius=10, shapeName=dyn_jacket_l_b_06_capsule, center=<10 0 6.25>, rotation=<rot 0 0 90>}
  {
   shapeType=Capsule
   length=7.3506
   radius=8.6279
   shapeName=dyn_jacket_l_b_06_capsule
   center=<10.3689 15.676 2.2889>
   rotation=<rot 0 0.0008 130.0002>
  }
 ]
 pelvis [{shapeType=Capsule, length=5.0796, radius=10.8527, center=<1.0428 0 0>}]
 root [{
   shapeType=Capsule
   length=119.2877
   radius=117.223
   shapeName=root_capsule
   center=<-0.0001 -33.0451 -111.3162>
   rotation=<rot -90 -0.0001 0>
  }]
 thigh_l [{shapeType=Capsule, length=28.3101, radius=8.9843, center=<-22.5663 0.0003 -0.8284>, rotation=<rot -90 -0.001 0>}]
 thigh_r [{shapeType=Capsule, length=28.3101, radius=8.9843, center=<22.5663 -0.0003 0.8284>, rotation=<rot -90 179.999 0>}]
}
constraints [
 {bone1Name=dyn_jacket_c_02, bone2Name=dyn_jacket_c_02_offset, swing1LimitDegrees=15, swing2LimitDegrees=40, twistLimitDegrees=1}
 {bone1Name=dyn_jacket_c_03, bone2Name=dyn_jacket_c_02, swing1LimitDegrees=15, swing2LimitDegrees=40, twistLimitDegrees=1}
 {bone1Name=dyn_jacket_c_04, bone2Name=dyn_jacket_c_03, swing1LimitDegrees=15, swing2LimitDegrees=40, twistLimitDegrees=1}
 {bone1Name=dyn_jacket_c_05, bone2Name=dyn_jacket_c_04, swing1LimitDegrees=15, swing2LimitDegrees=40, twistLimitDegrees=1}
 {bone1Name=dyn_jacket_c_06, bone2Name=dyn_jacket_c_05, swing1LimitDegrees=15, swing2LimitDegrees=40, twistLimitDegrees=1}
 {bone1Name=dyn_jacket_l_a_03, bone2Name=dyn_jacket_l_a_02, swing1LimitDegrees=30, swing2LimitDegrees=50, twistLimitDegrees=3}
 {bone1Name=dyn_jacket_l_a_04, bone2Name=dyn_jacket_l_a_03, swing1LimitDegrees=30, swing2LimitDegrees=50, twistLimitDegrees=3}
 {bone1Name=dyn_jacket_l_a_05, bone2Name=dyn_jacket_l_a_04, swing1LimitDegrees=30, swing2LimitDegrees=50, twistLimitDegrees=3}
 {bone1Name=dyn_jacket_l_a_06, bone2Name=dyn_jacket_l_a_05, swing1LimitDegrees=30, swing2LimitDegrees=50, twistLimitDegrees=3}
 {bone1Name=dyn_jacket_l_b_02, bone2Name=dyn_jacket_l_b_02_offset, swing1LimitDegrees=15, swing2LimitDegrees=30, twistLimitDegrees=3}
 {bone1Name=dyn_jacket_l_b_03, bone2Name=dyn_jacket_l_b_02, swing1LimitDegrees=15, swing2LimitDegrees=30, twistLimitDegrees=3}
 {bone1Name=dyn_jacket_l_b_04, bone2Name=dyn_jacket_l_b_03, swing1LimitDegrees=15, swing2LimitDegrees=30, twistLimitDegrees=3}
 {bone1Name=dyn_jacket_l_b_05, bone2Name=dyn_jacket_l_b_04, swing1LimitDegrees=15, swing2LimitDegrees=30, twistLimitDegrees=3}
 {bone1Name=dyn_jacket_l_b_06, bone2Name=dyn_jacket_l_b_05, swing1LimitDegrees=15, swing2LimitDegrees=30, twistLimitDegrees=3}
 {bone1Name=dyn_jacket_r_a_02, bone2Name=dyn_jacket_r_a_02_offset, swing1LimitDegrees=30, swing2LimitDegrees=50, twistLimitDegrees=3}
 {bone1Name=dyn_jacket_r_a_03, bone2Name=dyn_jacket_r_a_02, swing1LimitDegrees=30, swing2LimitDegrees=50, twistLimitDegrees=3}
 {bone1Name=dyn_jacket_r_a_04, bone2Name=dyn_jacket_r_a_03, swing1LimitDegrees=30, swing2LimitDegrees=50, twistLimitDegrees=3}
 {bone1Name=dyn_jacket_r_a_05, bone2Name=dyn_jacket_r_a_04, swing1LimitDegrees=30, swing2LimitDegrees=50, twistLimitDegrees=3}
 {bone1Name=dyn_jacket_r_a_06, bone2Name=dyn_jacket_r_a_05, swing1LimitDegrees=30, swing2LimitDegrees=50, twistLimitDegrees=3}
 {bone1Name=dyn_jacket_r_b_02, bone2Name=dyn_jacket_r_b_02_offset, swing1LimitDegrees=15, swing2LimitDegrees=30, twistLimitDegrees=3}
 {bone1Name=dyn_jacket_r_b_03, bone2Name=dyn_jacket_r_b_02, swing1LimitDegrees=15, swing2LimitDegrees=30, twistLimitDegrees=3}
 {bone1Name=dyn_jacket_r_b_05, bone2Name=dyn_jacket_r_b_04, swing1LimitDegrees=15, swing2LimitDegrees=30, twistLimitDegrees=3}
 {bone1Name=dyn_jacket_r_b_04, bone2Name=dyn_jacket_r_b_03, swing1LimitDegrees=15, swing2LimitDegrees=30, twistLimitDegrees=3}
 {bone1Name=dyn_jacket_l_a_02, bone2Name=dyn_jacket_l_a_02_offset, swing1LimitDegrees=30, swing2LimitDegrees=50, twistLimitDegrees=3}
 {bone1Name=dyn_jacket_c_02_offset, bone2Name=pelvis, swing1LimitDegrees=15, swing2LimitDegrees=15, twistLimitDegrees=1}
 {bone1Name=dyn_jacket_l_a_02_offset, bone2Name=pelvis, swing1LimitDegrees=15, swing2LimitDegrees=15, twistLimitDegrees=1}
 {bone1Name=dyn_jacket_l_b_02_offset, bone2Name=pelvis, swing1LimitDegrees=15, swing2LimitDegrees=15, twistLimitDegrees=1}
 {bone1Name=dyn_jacket_r_a_02_offset, bone2Name=pelvis, swing1LimitDegrees=15, swing2LimitDegrees=15, twistLimitDegrees=1}
 {bone1Name=dyn_jacket_r_b_02_offset, bone2Name=pelvis, swing1LimitDegrees=15, swing2LimitDegrees=15, twistLimitDegrees=1}
 {bone1Name=dyn_jacket_r_b_06, bone2Name=dyn_jacket_r_b_05, swing1LimitDegrees=15, swing2LimitDegrees=30, twistLimitDegrees=3}
 {
  bone1Name=dyn_jacket_l_b_05
  bone2Name=dyn_jacket_l_a_05
  swing1LimitDegrees=45
  swing1Motion=Free
  swing2LimitDegrees=45
  swing2Motion=Free
  twistLimitDegrees=45
  twistMotion=Free
 }
 {
  bone1Name=dyn_jacket_r_b_05
  bone2Name=dyn_jacket_r_a_05
  swing1LimitDegrees=45
  swing1Motion=Free
  swing2LimitDegrees=45
  swing2Motion=Free
  twistLimitDegrees=45
  twistMotion=Free
 }
 {
  bone1Name=dyn_jacket_l_b_06
  bone2Name=dyn_jacket_l_a_06
  swing1LimitDegrees=45
  swing1Motion=Free
  swing2LimitDegrees=45
  swing2Motion=Free
  twistLimitDegrees=45
  twistMotion=Free
 }
 {
  bone1Name=dyn_jacket_r_a_06
  bone2Name=dyn_jacket_r_b_06
  swing1LimitDegrees=45
  swing1Motion=Free
  swing2LimitDegrees=45
  swing2Motion=Free
  twistLimitDegrees=45
  twistMotion=Free
 }
]
