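# Post-process style Anim Blueprint for the CosmicTrellis male medium body cosmetic, adding procedural and physical secondary motion.
# It copies the attached parent's pose through a Control Rig, then a spine_01-based RigidBody simulation.
# Four spring-driven, angle-limited AnimDynamics nodes then simulate left and right collar bones relative to spine bones.
# Each update stores gravity and RigidBody settings from a parameter-asset function, plus a combined baseball/knee sliding flag.
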AnimBlueprint /BRCosmetics/Characters/Player/Male/Medium/Bodies/M_MED_CosmicTrellis/Meshes/M_MED_CosmicTrellis_AnimBP
target_skeleton=&/M_MED_CosmicTrellis_Skeleton
graphs [
 AnimGraph {nodes [
   n1 = Control Rig!(Source=Copy Pose From Mesh() {bUseAttachedParent=true}, Alpha, IsSliding:bool=false, IsCrouch:bool=false) -> (Pose) {controlRigAssetReference {controlRigAsset=&/M_MED_CosmicTrellis_CtrlRig}}
   n2 = RigidBody!(
    ComponentPose=Local To Component(LocalPose=@n1)
    Alpha
    OverrideWorldGravity:Vector="0,0,0"
    ExternalForce
    SimSpaceSettings:Sim Space Settings Structure=(WorldAlpha=0,VelocityScaleZ=1,DampingAlpha=1,MaxLinearVelocity=10000,MaxAngularVelocity=10000,MaxLinearAcceleration=10000,MaxAngularAcceleration=10000,ExternalLinearDragV=(X=0,Y=0,Z=0),ExternalLinearVelocity=(X=0,Y=0,Z=0),ExternalAngularVelocity=(X=0,Y=0,Z=0))
   ) -> (Pose) {
    bOverrideWorldGravity=true
    overridePhysicsAsset=&/M_MED_CosmicTrellis_Physics
    simulationSpace=BaseBoneSpace
    baseBoneRef {boneName=spine_01}
   }
   n3 = Anim Dynamics - Bone: dyn_collar_l_a_01(ComponentPose=@n2, Alpha=1, ExternalForce="0,0,0") {
    angularDampingOverride=0.9999
    angularSpringConstant=2500
    bAngularSpring=true
    bOverrideAngularDamping=true
    bUseGravityOverride=true
    linearDampingOverride=0.999
    linearSpringConstant=0.9999
    simulationSpace=RootRelative
    boundBone {boneName=dyn_collar_l_a_01}
    componentAppliedLinearAccClamp=<2000 2000 2000>
    componentLinearAccScale=<0.3 0.3 0.3>
    componentLinearVelScale=<0.2 0.2 0.2>
    gravityOverride {z=-700}
    physicsBodyDefinitions [{
      boundBone {boneName=dyn_collar_l_a_01}
      boxExtents=<20 20 10>
      constraintSetup {angularLimitsMax=<0 30 1>, angularLimitsMin=<0 -10 -1>, angularTarget=<0 0 0>}
      localJointOffset=<-3 0 -1>
     }]
    relativeSpaceBone {boneName=spine_02}
    simSpaceSettings {maxAngularAcceleration=40, maxAngularVelocity=8, simSpaceAngularAlpha=1}
   }
   n4 = Anim Dynamics - Bone: dyn_collar_r_a_01(ComponentPose=@n3, Alpha=1, ExternalForce="0,0,0") {
    angularDampingOverride=0.9999
    angularSpringConstant=2500
    bAngularSpring=true
    bOverrideAngularDamping=true
    bUseGravityOverride=true
    linearDampingOverride=0.999
    linearSpringConstant=0.9999
    simulationSpace=RootRelative
    boundBone {boneName=dyn_collar_r_a_01}
    componentAppliedLinearAccClamp=<2000 2000 2000>
    componentLinearAccScale=<0.3 0.3 0.3>
    componentLinearVelScale=<0.2 0.2 0.2>
    gravityOverride {z=-700}
    physicsBodyDefinitions [{
      boundBone {boneName=dyn_collar_r_a_01}
      boxExtents=<20 20 10>
      constraintSetup {angularLimitsMax=<0 30 1>, angularLimitsMin=<0 -10 -1>, angularTarget=<0 0 0>}
      localJointOffset=<3 0 0.5>
     }]
    relativeSpaceBone {boneName=spine_02}
    simSpaceSettings {maxAngularAcceleration=40, maxAngularVelocity=8, simSpaceAngularAlpha=1}
   }
   n5 = Anim Dynamics - Bone: dyn_collar_l_b_01(ComponentPose=@n4, Alpha=1, ExternalForce="0,0,0") {
    angularDampingOverride=0.9999
    angularSpringConstant=3500
    bAngularSpring=true
    bGravityOverrideInSimSpace=true
    bOverrideAngularDamping=true
    bUseGravityOverride=true
    linearDampingOverride=0.9999
    linearSpringConstant=0.9999
    simulationSpace=RootRelative
    boundBone {boneName=dyn_collar_l_b_01}
    componentAppliedLinearAccClamp=<2500 2500 2500>
    componentLinearAccScale=<0.2 0.2 0.2>
    componentLinearVelScale=<0.2 0.2 0.2>
    gravityOverride {z=-700}
    physicsBodyDefinitions [{
      boundBone {boneName=dyn_collar_l_b_01}
      boxExtents=<35 35 10>
      constraintSetup {angularLimitsMax=<0 0 0>, angularLimitsMin=<0 -5 0>, angularTarget=<0 0 0>}
      localJointOffset=<1.5 0 0>
     }]
    relativeSpaceBone {boneName=spine_04}
    simSpaceSettings {maxAngularAcceleration=15, maxAngularVelocity=3, simSpaceAngularAlpha=1}
   }
   n6 = Anim Dynamics - Bone: dyn_collar_r_b_01(ComponentPose=@n5, Alpha=1, ExternalForce="0,0,0") {
    angularDampingOverride=0.9999
    angularSpringConstant=3500
    bAngularSpring=true
    bGravityOverrideInSimSpace=true
    bOverrideAngularDamping=true
    bUseGravityOverride=true
    linearDampingOverride=0.9999
    linearSpringConstant=0.9999
    simulationSpace=RootRelative
    boundBone {boneName=dyn_collar_r_b_01}
    componentAppliedLinearAccClamp=<2500 2500 2500>
    componentLinearAccScale=<0.2 0.2 0.2>
    componentLinearVelScale=<0.2 0.2 0.2>
    gravityOverride {z=-700}
    physicsBodyDefinitions [{
      boundBone {boneName=dyn_collar_r_b_01}
      boxExtents=<35 35 10>
      constraintSetup {angularLimitsMax=<0 0 0>, angularLimitsMin=<0 -5 0>, angularTarget=<0 0 0>}
      localJointOffset=<-1.5 0 0>
     }]
    relativeSpaceBone {boneName=spine_04}
    simSpaceSettings {maxAngularAcceleration=15, maxAngularVelocity=3, simSpaceAngularAlpha=1}
   }
   Output Pose(Result=Component To Local(ComponentPose=@n6)) {layerGroup=None}
  ]}
 EventGraph {nodes [
   Event BlueprintUpdateAnimation()
   > n1 = "Gravity Override State Function (Parameter Asset) (Nativized)"(
    CharacterPartAnimInstance=Self-Reference()
    GameTimeInSeconds=GetGameTimeInSeconds()
    Parameters=&/M_MED_CosmicTrellis_Parameters
   )
   > Set Gravity Override(Gravity Override=@n1)
   > Set RBAN Settings Out(RBAN Settings Out=@n1.OutRigidBodyAnimNodeParameters)
   > Set IsSliding(IsSliding:bool=OR Boolean(A=Get bIsBaseballSliding(), B=Get bIsKneeSliding())) -> (Output_Get:bool)
  ]}
]
variables [
 Gravity Override {type=Vector}
 IsSliding {type=boolean}
 RBAN Settings Out {type=FortRigidBodyAnimNodeParameters}
]
