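# Physics asset for the CosmicTrellis medium male player body, named as the collider set for its cape simulation.
# Capsules cover the pelvis, spine chain, neck and both clavicles, with a sphere on the head.
# An oversized capsule on the dyn_jacket_c_05 bone adds a broad collision volume beyond the core body.
# Arm, leg and foot bodies carry no shapes, and no constraints link any bodies.
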PhysicsAsset /BRCosmetics/Characters/Player/Male/Medium/Bodies/M_MED_CosmicTrellis/Meshes/M_MED_CosmicTrellis_CapeColliders
bodies {
 calf_l []
 calf_r []
 clavicle_l [{shapeType=Capsule, length=9.0192, radius=9.8871, center=<9.4266 -1.1806 -2.4977>, rotation=<rot 69.3297 1.4476 176.9462>}]
 clavicle_r [{shapeType=Capsule, length=9.0192, radius=9.8871, center=<-9.4266 1.1806 2.4977>, rotation=<rot 69.3297 1.4476 -3.0538>}]
 dyn_jacket_c_05 [{
   shapeType=Capsule
   length=32.1726
   radius=34.5848
   shapeName=dyn_jacket_c_05_capsule
   center=<0.3974 -0.0981 -26.0271>
   rotation=<rot 0.2112 -179.8335 -90>
  }]
 foot_l []
 foot_r []
 head [{shapeType=Sphere, radius=17.3125, center=<11.028 -0.174 0>}]
 lowerarm_l []
 lowerarm_r []
 neck_02 [{shapeType=Capsule, length=13.4375, radius=7.3175, center=<-0.2642 -0.3408 0>, rotation=<rot 90 -171.1193 -175.3978>}]
 pelvis [{shapeType=Capsule, length=33.6456, radius=14.5388, center=<-2.5599 -5.6779 0>}]
 spine_01 [{shapeType=Capsule, length=30.9438, radius=13.5586, center=<-1.7323 -4.7607 0>}]
 spine_02 [{shapeType=Capsule, length=24.8264, radius=13.6322, center=<-1.5525 -4.3168 0.0001>}]
 spine_03 [{shapeType=Capsule, length=19.7107, radius=13.7031, center=<0.3619 -1.7376 0.0001>}]
 spine_04 [{shapeType=Capsule, length=9.4404, radius=14.1325, center=<2.9313 -1.2057 0.0002>}]
 spine_05 [{shapeType=Capsule, length=13.9518, radius=13.9425, center=<-0.1322 -1.5374 0>}]
 thigh_l []
 thigh_r []
 upperarm_l []
 upperarm_r []
}
constraints []
