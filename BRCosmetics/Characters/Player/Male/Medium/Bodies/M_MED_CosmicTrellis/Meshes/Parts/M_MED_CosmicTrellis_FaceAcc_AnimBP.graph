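# Post-copy secondary-motion AnimBP for the CosmicTrellis male face accessory skeleton, adding physics jiggle to its hair bones.
# Copies the attached parent mesh's pose, then chains four Anim Dynamics nodes for two front, one left and one right hair bone.
# Each simulates root-relative with angular springs, tight angular limits, overridden gravity and damping, and scaled-down component motion.
# The event graph holds no logic.
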AnimBlueprint /BRCosmetics/Characters/Player/Male/Medium/Bodies/M_MED_CosmicTrellis/Meshes/Parts/M_MED_CosmicTrellis_FaceAcc_AnimBP
target_skeleton=&/M_MED_CosmicTrellis_FaceAcc_Skeleton
graphs [
 AnimGraph {nodes [
   n1 = Anim Dynamics - Bone: dyn_hair_fr_A_01(ComponentPose=Local To Component(LocalPose=Copy Pose From Mesh() {bUseAttachedParent=true}), Alpha=1, ExternalForce="0,0,0") {
    angularDampingOverride=1
    angularSpringConstant=2000
    bAngularSpring=true
    bOverrideAngularDamping=true
    bUseGravityOverride=true
    gravityScale=0.35
    simulationSpace=RootRelative
    alphaBoolBlend {blendOption=Linear}
    boundBone {boneName=dyn_hair_fr_A_01}
    componentAppliedLinearAccClamp=<10000 10000 10000>
    componentLinearAccScale=<0.25 0.25 0.25>
    componentLinearVelScale=<0.25 0.25 0.8>
    gravityOverride {z=-500}
    physicsBodyDefinitions [{
      boundBone {boneName=dyn_hair_fr_A_01}
      boxExtents=<12 12 12>
      constraintSetup {angularLimitsMax=<0 0 15>, angularLimitsMin=<0 -15 -10>, angularTarget=<1 0 0>}
      localJointOffset=<1 0 0>
     }]
   }
   n2 = Anim Dynamics - Bone: dyn_hair_fr_B_01(ComponentPose=@n1, Alpha=1, ExternalForce="0,0,0") {
    angularDampingOverride=0.9999
    angularSpringConstant=1750
    bAngularSpring=true
    bOverrideAngularDamping=true
    bUseGravityOverride=true
    gravityScale=0.35
    simulationSpace=RootRelative
    alphaBoolBlend {blendOption=Linear}
    boundBone {boneName=dyn_hair_fr_B_01}
    componentAppliedLinearAccClamp=<10000 10000 10000>
    componentLinearAccScale=<0.2 0.2 0.8>
    componentLinearVelScale=<0.2 0.2 0.8>
    gravityOverride {z=-550}
    physicsBodyDefinitions [{
      boundBone {boneName=dyn_hair_fr_B_01}
      boxExtents=<13 12 12>
      constraintSetup {angularLimitsMax=<0 0 12.5>, angularLimitsMin=<0 -10 -12.5>, angularTarget=<1 0 0>}
      localJointOffset=<0.75 0 0>
     }]
   }
   n3 = Anim Dynamics - Bone: dyn_hair_lt_A_01(ComponentPose=@n2, Alpha=1, ExternalForce="0,0,0") {
    angularDampingOverride=1
    angularSpringConstant=2000
    bAngularSpring=true
    bOverrideAngularDamping=true
    bUseGravityOverride=true
    gravityScale=0.35
    simulationSpace=RootRelative
    alphaBoolBlend {blendOption=Linear}
    boundBone {boneName=dyn_hair_lt_A_01}
    componentAppliedLinearAccClamp=<10000 10000 10000>
    componentLinearAccScale=<0.2 0.2 0.8>
    componentLinearVelScale=<0.2 0.2 0.8>
    gravityOverride {z=-500}
    physicsBodyDefinitions [{
      boundBone {boneName=dyn_hair_lt_A_01}
      boxExtents=<12 12 12>
      constraintSetup {angularLimitsMax=<0 10 10>, angularLimitsMin=<0 0 -10>, angularTarget=<1 0 0>}
      localJointOffset=<1 0 0>
     }]
   }
   n4 = Anim Dynamics - Bone: dyn_hair_rt_A_01(ComponentPose=@n3, Alpha=1, ExternalForce="0,0,0") {
    angularDampingOverride=1
    angularSpringConstant=2000
    bAngularSpring=true
    bOverrideAngularDamping=true
    bUseGravityOverride=true
    gravityScale=0.35
    simulationSpace=RootRelative
    alphaBoolBlend {blendOption=Linear}
    boundBone {boneName=dyn_hair_rt_A_01}
    componentAppliedLinearAccClamp=<10000 10000 10000>
    componentLinearAccScale=<0.2 0.2 0.8>
    componentLinearVelScale=<0.2 0.2 0.8>
    gravityOverride {z=-500}
    physicsBodyDefinitions [{
      boundBone {boneName=dyn_hair_rt_A_01}
      boxExtents=<13 12.5 12>
      constraintSetup {angularLimitsMax=<0 10 10>, angularLimitsMin=<0 0 -10>, angularTarget=<1 0 0>}
      localJointOffset=<0.5 0 0>
     }]
   }
   Output Pose(Result=Component To Local(ComponentPose=@n4)) {layerGroup=None}
  ]}
 EventGraph {nodes [Event BlueprintUpdateAnimation()]}
]
variables []
